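# WBP_UEFN_Block_Background_Base is a widget base class for UEFN button background blocks, derived from WBP_UEFN_Block_Base, with no widgets of its own.
# It declares functions to set background and outline brushes and materials, all with empty bodies, and an unconnected PreConstruct event.
# Variables hold background and outline material instances, both defaulting to MI_UEFN_UI_BackgroundColor, plus a dynamic material slot for each.
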WidgetBlueprint /FortUILibrary/UGC/Widgets/Buttons/Blocks/WBP_UEFN_Block_Background_Base
graphs [
 EventGraph {nodes [Event PreConstruct()]}
 SetBackgroundBrush {nodes [SetBackgroundBrush!() -> (then, InBrush:"Slate Brush Structure (by ref)") {is=title K2Node_FunctionEntry}]}
 SetBackgroundMaterial {nodes [SetBackgroundMaterial!() -> (then, Material Instance:"Material Instance Object Reference (by ref)") {is=title K2Node_FunctionEntry}]}
 SetOutlineBrush {nodes [SetOutlineBrush!() -> (then, In Brush:"Slate Brush Structure (by ref)") {is=title K2Node_FunctionEntry}]}
 SetOutlineMaterial {nodes [SetOutlineMaterial!() -> (then, Material Instance:"Material Instance Object Reference (by ref)") {is=title K2Node_FunctionEntry}]}
]
variables [
 BackgroundDynamicMaterial {type=MaterialInstanceDynamic}
 BackgroundMaterial {type=MaterialInstance, default=/FortUILibrary/UGC/MaterialInstances/MI_UEFN_UI_BackgroundColor}
 OutlineDynamicMaterial {type=MaterialInstanceDynamic}
 OutlineMaterial {type=MaterialInstance, default=&MI_UEFN_UI_BackgroundColor}
]
widgets {
 info {
  inheritedWidgetCount=0
  parentClass=&/WBP_UEFN_Block_Base.WBP_UEFN_Block_Base_C
  rootWidgetClass=null
  widgetCount=0
 }
 widgets []
}
